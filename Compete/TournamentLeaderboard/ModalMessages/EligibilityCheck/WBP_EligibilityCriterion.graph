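# List entry widget for one criterion in the tournament eligibility check modal, showing a state icon, a name and an auto-collapsing explanation.
# When the list item is set, it casts the object to VM_EligibilityCheckCriteria and assigns it as its view model.
# SetName and SetDescription fill the two text blocks.
# SetState and SetState_Updated both pick the icon material from a pass flag, sharing the check mark but differing in the fail material.
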WidgetBlueprint /Compete/TournamentLeaderboard/ModalMessages/EligibilityCheck/WBP_EligibilityCriterion
graphs [
 EventGraph {nodes [
   n1 = Event OnListItemObjectSet()
   > n2 = Cast To VM_EligibilityCheckCriteria!(execute, Object=@n1.ListItemObject) -> (then, CastFailed, AsVM Eligibility Check Criteria:VM Eligibility Check Criteria Object Reference) {is=title K2Node_DynamicCast}
   > Set VM_EligibilityCheckCriteria(VM_EligibilityCheckCriteria:VM Eligibility Check Criteria Object Reference=@n2) -> (Output_Get:VM Eligibility Check Criteria Object Reference)
  ]}
 SetDescription {nodes [
   n1 = SetDescription!() -> (then, InText:Text) {is=title K2Node_FunctionEntry}
   > "SetText (Text)"(self=Get ExplanationText() -> (ExplanationText:Common Text Object Reference), InText=@n1)
  ]}
 SetName {nodes [
   n1 = SetName!() -> (then, InText:Text) {is=title K2Node_FunctionEntry}
   > "SetText (Text)"(self=Get CritetionName() -> (CritetionName:Common Text Object Reference), InText=@n1)
  ]}
 SetState {nodes [
   n1 = SetState!() -> (then, Passed:bool=~) {is=title K2Node_FunctionEntry}
   > SetBrushFromMaterial(
    self=Get State() -> (State:Image Object Reference)
    Material=Select(
     Option 0:Material Interface Object Reference=/Compete/MaterialInstances/MI_UI_Compete_EligibilityState
     Option 1:Material Interface Object Reference=/FortUILibrary/Textures/MaterialInstances/MI_UIKit_Icon_Check
     Index=@n1
    ) -> (ReturnValue:Material Interface Object Reference)
   )
  ]}
 SetState_Updated {nodes [
   n1 = SetState_Updated!() -> (then, Passed:bool=~) {is=title K2Node_FunctionEntry}
   > SetBrushFromMaterial(
    self=Get State()
    Material=Select(
     Option 0=/Compete/MaterialInstances/MI_UI_EligilityWarning
     Option 1=&MI_UIKit_Icon_Check
     Index=@n1
    )
   )
  ]}
]
variables [
 bPassedState {type=boolean}
 CriterionName {type=string, default="[PH] Text"}
]
widgets {
 info {inheritedWidgetCount=0, parentClass=&CommonUserWidget, rootWidgetClass=&GridPanel, widgetCount=4}
 widgets [
  GridPanel_23 {widgetClassPath=&GridPanel, widget=GridPanel_23, properties {columnFill [0,1]}}
  State {
   widgetClassPath=&Image
   bIsVariable=true
   parent=GridPanel_23
   widget=State
   properties {brush {
     bIsDynamicallyLoaded=false
     drawAs=Image
     imageType=FullColor
     mirroring=NoMirror
     resourceName=None
     resourceObject=&MI_UIKit_Icon_Check
     tiling=NoTile
     imageSize=<32 32>
     margin=<ltrb 0 0 0 0>
     outlineSettings {
      bUseBrushTransparency=false
      roundingType=HalfHeightRadius
      width=0
      color {colorUseRule=UseColor_Specified, specifiedColor=<rgba 0 0 0 0>}
      cornerRadii=<3.0446 0 0 0>
     }
     tintColor {colorUseRule=UseColor_Specified, specifiedColor=<rgba 1 1 1 1>}
     uVRegion {bIsValid=false, max=<0 0>, min=<0 0>}
    }}
   slot {class=GridSlot, properties {horizontalAlignment=HAlign_Center, verticalAlignment=VAlign_Top}}
  }
  CritetionName {
   widgetClassPath=&CommonTextBlock
   bIsVariable=true
   parent=GridPanel_23
   widget=CritetionName
   properties {
    text=Text Block
    wrapTextAt=635
    font {
     bForceMonospaced=false
     bMaterialIsStencil=false
     fontMaterial=null
     fontObject=&HeadingNow
     letterSpacing=0
     monospacedWidth=1
     size=18
     skewAmount=0
     typefaceFontName=75 Medium
     outlineSettings {
      bApplyOutlineToDropShadows=false
      bMiteredCorners=false
      bSeparateFillAlpha=false
      outlineMaterial=null
      outlineSize=0
      outlineColor=<rgba 0 0 0 1>
     }
    }
    shadowOffset=<0 0>
   }
   slot {class=GridSlot, properties {column=1, padding=<ltrb 16 0 0 0>}}
  }
  ExplanationText {
   widgetClassPath=&CommonTextBlock
   bIsVariable=true
   parent=GridPanel_23
   widget=ExplanationText
   properties {
    autoWrapText=true
    bAutoCollapseWithEmptyText=true
    colorAndOpacity {colorUseRule=UseColor_Specified, specifiedColor=<rgba 0.5776 0.5776 0.5776 1>}
    font {
     bForceMonospaced=false
     bMaterialIsStencil=false
     fontMaterial=null
     fontObject=&HeadingNow
     letterSpacing=0
     monospacedWidth=1
     size=18
     skewAmount=0
     typefaceFontName=74 Regular
     outlineSettings {
      bApplyOutlineToDropShadows=false
      bMiteredCorners=false
      bSeparateFillAlpha=false
      outlineMaterial=null
      outlineSize=0
      outlineColor=<rgba 0 0 0 1>
     }
    }
    shadowOffset=<0 0>
   }
   slot {class=GridSlot, properties {column=1, row=1, padding=<ltrb 16 5 0 0>}}
  }
 ]
}
