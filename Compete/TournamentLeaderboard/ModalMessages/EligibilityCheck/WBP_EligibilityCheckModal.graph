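# Activatable UIKit dialog that shows a player's tournament eligibility checks, with pass/fail criteria, explanation and a red error banner.
# SetEligibilityCriteria turns requirement items into list-view viewmodels, omitting EULA and skip-listed types, and pulls banner text for failures from RedMsgList.
# Enum-driven selects supply each title and description, with rank rules built from the requirements viewmodel.
# Custom navigation scrolls the list, gamepad input refocuses the dialog, and open/close events fire analytics.
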WidgetBlueprint /Compete/TournamentLeaderboard/ModalMessages/EligibilityCheck/WBP_EligibilityCheckModal
graphs [
 DebugPassedRequirements {nodes [
   n1 = DebugPassedRequirements!() -> (then, InputPin:bool=~, Index:EFortTournamentRequirementCheckType Enum=Invalid) {is=title K2Node_FunctionEntry}
   > Return Node!(
    execute
    ReturnValue:bool=Select!(
     Invalid:bool=false
     SystemFeatureFlag_CryptoProcessor:bool=@n1
     SystemFeatureFlag_SecureBoot:bool=@n1
     SystemFeatureFlag_KernelCodeIntegrity:bool=@n1
     SystemFeatureFlag_IOMMU:bool=@n1
     SystemFeatureFlag_StableOSVersion:bool=@n1
     SystemFeatureFlag_StrictInjectionPolicy:bool=@n1
     SystemFeatureFlag_StrictVirtualizationPolicy:bool=@n1
     SystemFeatureFlag_AndroidSecureDevice:bool=@n1
     SystemFeatureFlag_IOSSecureDevice:bool=@n1
     SystemFeatureFlag_IOSMinimumOSVersion:bool=@n1
     AccountLevel:bool=@n1
     MultiFactorAuthentication:bool=@n1
     HabaneroRank:bool=@n1
     QualifyingRound:bool=@n1
     EULA:bool=true
     TotalNumberOfPlayedMatches:bool=@n1
     SeasonNumberOfPlayedMatches:bool=@n1
     Age:bool=@n1
     RegionLock:bool=@n1
     AccountLock:bool=@n1
     TeamLock:bool=@n1
     TokenRestrictions:bool=@n1
     SuspiciousBehavior:bool=@n1
     PermanentTournamentBan:bool=false
     InputType:bool=@n1
     TimeAlive:bool=@n1
     PowerRankings:bool=false
     Index:EFortTournamentRequirementCheckType Enum=@n1.Index
    ) -> (ReturnValue:bool)
   )
  ]}
 DoCustomNavigation {nodes [
   n1 = DoCustomNavigation!() -> (then, Navigation:EUINavigation Enum) {is=title K2Node_FunctionEntry}
   > n2 = Switch on EUINavigation!(execute, Selection:EUINavigation Enum=@n1) -> (Left:Exec, Right:Exec, Up:Exec, Down:Exec, Next:Exec, Previous:Exec, Invalid:Exec) {is=title K2Node_SwitchEnum}
   n3 = GetScrollOffset!(self:List View Base Object Reference=Get CriteriaListView() -> (CriteriaListView:Common List View Object Reference)) -> (ReturnValue:float)
   n4 = SetScrollOffset!(execute=@n2.Up, self:List View Base Object Reference=Get CriteriaListView(), InScrollOffset:float=float - float(A=@n3, B=1)) -> (then)
   Return Node(execute=[@n4.then, @n5.then], ReturnValue:Widget Object Reference)
   n5 = SetScrollOffset(execute=@n2.Down, self=Get CriteriaListView(), InScrollOffset=float + float(A=@n3, B=1))
  ]}
 EventGraph {nodes [
   n1 = OnInputTypeChanged!() -> (OutputDelegate, then, bNewInputType:ECommonInputType Enum) {is=title K2Node_CustomEvent}
   > n2 = Switch on ECommonInputType!(execute, Selection:ECommonInputType Enum=@n1.bNewInputType) -> (MouseAndKeyboard:Exec, Gamepad:Exec, Touch:Exec, Count:Exec) {is=title K2Node_SwitchEnum}
   SetFocus(execute=@n2.Gamepad, self=Get Dialog_Base() -> (Dialog_Base:WBP UIKit Dialog Internal Object Reference))
   OnButtonClicked!() -> (OutputDelegate, then, ButtonIndex:int) {is=title K2Node_CustomEvent}
   > Branch(Condition=Get bIsActive() -> (bIsActive:bool))
   > DeactivateWidget()
   > n3 = Cast To FortPlayerController(Object=GetOwningPlayer())
   > "Fire Event Element Interacted (Non-Localized)"!(
    execute
    FortPC:Fort Player Controller Object Reference=@n3
    ElementName:String=CloseButton
    AdditionalInfo:Flow Analytics Additional Info Structure
   ) -> (then)
   n4 = Event PreConstruct()
   > n5 = Branch(Condition=@n4.IsDesignTime)
   UpdateWidgetStyleTags!(
    execute=[@n8.then, @n5.then]
    InWidget:Widget Object Reference=Get ScrollBox_Text!(self:WBP UIKit Dialog Internal Object Reference=Get Dialog_Base()) -> (ScrollBox_Text:Scroll Box Object Reference)
    InAddedTags:Array of Gameplay Tag Structures=Make Array("[0]"=MakeLiteralGameplayTag!(Value:Gameplay Tag Structure=(TagName="StyleSheet.Style.UIKit")) -> (ReturnValue:Gameplay Tag Structure))
    InRemovedTags:Array of Gameplay Tag Structures=Make Array()
   ) -> (then, ReturnValue:bool)
   > n6 = Construct Object from Class(Class=&/VM_Dialog_EligibilityCheckModal.VM_Dialog_EligibilityCheckModal_C) -> (ReturnValue:VM Dialog Eligibility Check Modal Object Reference)
   > SetDialogData!(execute, self:WBP UIKit Dialog Internal Object Reference=Get Dialog_Base(), Data:UI Kit Dialog Viewmodel Object Reference=@n6) -> (then)
   n7 = SlotAsGridSlot(Widget=Get Slot_Content!(self:WBP UIKit Dialog Internal Object Reference=Get Dialog_Base()) -> (Slot_Content:Named Slot Object Reference))
   > SetPadding!(
    execute
    self:Grid Slot Object Reference=@n7
    InPadding_Left:float=80
    InPadding_Top:float=0
    InPadding_Right:float=80
    InPadding_Bottom:float=0
   ) -> (then)
   > SetLayer!(execute, self:Grid Slot Object Reference=@n7, InLayer:int=30) -> (then)
   n8 = "SetText (Text)"(execute=@n5.else, self=Get ErrorMessage() -> (ErrorMessage:Common Text Object Reference))
   Event On Activated()
   n9 = Get Dialog_Base()
   > Unbind all Events from Dialog Button Clicked Dispatcher!(execute:Exec, self:WBP UIKit Dialog Internal Object Reference=@n9) -> (then:Exec) {is=title K2Node_ClearDelegate}
   > Bind Event to Dialog Button Clicked Dispatcher(self:WBP UIKit Dialog Internal Object Reference=@n9, Delegate=Create Event()) {is=title K2Node_AddDelegate}
   > SetFocus(self=Get Dialog_Base())
   > Assign On Input Method Changed!(execute:Exec, self:Common Input Object Reference=Get CommonInputSubsystem(), Delegate:"Delegate (by ref)"=@n1) -> (then:Exec) {is=title K2Node_AssignDelegate}
   > Fire Event Sub View Opened!(execute, View:Widget Object Reference=Self-Reference(), AdditionalInfo:Flow Analytics Additional Info Structure, bIsOverlay:bool=true) -> (then)
  ]}
 GetRankRangeText {nodes [
   n1 = GetRankRangeText!() -> (then, RankKey:Fort Compete UIHabanero Tier Override Key Structure, bIsMin:bool=~) {is=title K2Node_FunctionEntry}
   > n2 = Branch(Condition=@n1.bIsMin)
   n3 = GetHabaneroDisplayDataOverride!(
    InDisplayData:Fort Habanero Display Data Object Reference=/Game/UI/Habanero/HabaneroDisplayData
    InOverrideData:Fort Compete UIHabanero Tier Override List Object Reference=/Game/UI/Habanero/DA_RankDisplayDataOverride
    InKey:"Fort Compete UIHabanero Tier Override Key Structure (by ref)"=@n1
   ) -> (
    OutData_ForegroundIcon:Material Instance Soft Object Reference
    OutData_ForegroundIconSmall:Material Instance Soft Object Reference
    OutData_BackgroundIcon:Material Instance Soft Object Reference
    OutData_BackgroundIconSmall:Material Instance Soft Object Reference
    OutData_WhiteIcon:Texture 2D Soft Object Reference
    OutData_SimpleIconSmall:Texture 2D Soft Object Reference
    OutData_Name:Text
    OutData_TierType:Fort Habanero Tier Type Data Object Reference
    OutData_AdditionalLeftIcon:Material Instance Soft Object Reference
    OutData_AdditionalRightIcon:Material Instance Soft Object Reference
    ReturnValue:bool
   )
   > Return Node!(execute, OutputText:Text=Format Text!(Format="LOCTEXT(\"{MinRank} or higher\")", MinRank:Text=@n3.OutData_Name) -> (Result))
   Return Node(execute=@n2.else, OutputText=Format Text!(Format="LOCTEXT(\"{MaxRank} or lower\")", MaxRank:Text=@n3.OutData_Name) -> (Result))
  ]}
 GetRankRequirementText {nodes [
   GetRankRequirementText() {is=title K2Node_FunctionEntry}
   > Set Output(Output:Text="LOCTEXT(\"Your Rank must meet one of the requirement(s) below:\\r\\n\")") -> (Output_Get:Text)
   > n1 = For Each Loop!(
    Exec
    Array:Array of Rank Requirement Line Display Structures=GetRequirementLines!(self:Tournament Requirements Modal VM Object Reference=Get FortTournamentRequirementsModalVM() -> (FortTournamentRequirementsModalVM:Tournament Requirements Modal VM Object Reference)) -> (ReturnValue:Array of Rank Requirement Line Display Structures)
   ) -> (
    LoopBody
    Array Element_GameModeDisplayName:Text
    Array Element_RankKey:Fort Compete UIHabanero Tier Override Key Structure
    Array Element_bIsMinRank:bool=~
    Array Element_SeasonName:Text
    Array Index
    Completed
   )
   n2 = SelectText(A="LOCTEXT(\"any season\")", B=@n1.Array Element_SeasonName, bPickA=TextIsEmpty(InText=@n1.Array Element_SeasonName))
   n3 = GetRankRangeText!(RankKey:Fort Compete UIHabanero Tier Override Key Structure=@n1.Array Element_RankKey, bIsMin:bool=@n1.Array Element_bIsMinRank) -> (OutputText:Text) {is=title K2Node_CallFunction}
   > Set Output(Output=Format Text!(
     Format="LOCTEXT(\"{Existing}{New}\")"
     Existing:Text=Get Output() -> (Output:Text)
     New:Text=Format Text!(Format="LOCTEXT(\"- {Game Mode}: {Rank Range} in {Season}\\r\\n\")", Game Mode:Text=@n1, Rank Range:Text=@n3, Season:Text=@n2) -> (Result)
    ) -> (Result))
   Return Node(
    execute=@n1.Completed
    OutputText=Format Text!(Format="LOCTEXT(\"{ConstructedText}\\r\\nCheck your rank by mode and season in the Career tab.\")", ConstructedText:Text=Get Output()) -> (Result)
   )
  ]}
 RequirementsDescriptionTexts {nodes [
   n1 = RequirementsDescriptionTexts!() -> (then, Arguments:"Tournament Requirement Item Arguments Structure (by ref)", RequirementType:EFortTournamentRequirementCheckType Enum=Invalid) {is=title K2Node_FunctionEntry}
   n2 = MakeLiteralText(Value="LOCTABLE(\"&/ST_EligibilityDescription\", \"NoTokens\")")
   > Return Node!(
    execute
    ReturnValue:Text=Select(
     Invalid:Text="LOCTABLE(\"&ST_EligibilityDescription\", \"InvalidReqCheck\")"
     SystemFeatureFlag_CryptoProcessor:Text="LOCTABLE(\"&ST_EligibilityDescription\", \"SysCryptoProc\")"
     SystemFeatureFlag_SecureBoot:Text="LOCTABLE(\"&ST_EligibilityDescription\", \"SysSeqBoot\")"
     SystemFeatureFlag_KernelCodeIntegrity:Text="LOCTABLE(\"&ST_EligibilityDescription\", \"SysFlagKernel\")"
     SystemFeatureFlag_IOMMU:Text="LOCTABLE(\"&ST_EligibilityDescription\", \"SysIOMMU\")"
     SystemFeatureFlag_StableOSVersion:Text="LOCTABLE(\"&ST_EligibilityDescription\", \"SysOSVersion\")"
     SystemFeatureFlag_StrictInjectionPolicy:Text="LOCTABLE(\"&ST_EligibilityDescription\", \"SysInjection\")"
     SystemFeatureFlag_StrictVirtualizationPolicy:Text="LOCTABLE(\"&ST_EligibilityDescription\", \"SysNotVM\")"
     SystemFeatureFlag_AndroidSecureDevice:Text="LOCTABLE(\"&ST_EligibilityDescription\", \"SysAndroidSecureDevice\")"
     SystemFeatureFlag_IOSSecureDevice:Text="LOCTABLE(\"&ST_EligibilityDescription\", \"SysIOSSecureDevice\")"
     SystemFeatureFlag_IOSMinimumOSVersion:Text="LOCTABLE(\"&ST_EligibilityDescription\", \"SysIOSMinimumOSVersion\")"
     AccountLevel:Text="LOCTABLE(\"&ST_EligibilityDescription\", \"AccountLevel\")"
     MultiFactorAuthentication:Text="LOCTABLE(\"&ST_EligibilityDescription\", \"MFA\")"
     HabaneroRank:Text=GetRankRequirementText!() -> (OutputText:Text) {is=title K2Node_CallFunction}
     QualifyingRound:Text="LOCTABLE(\"&ST_EligibilityDescription\", \"QualifyingRound\")"
     EULA:Text="LOCTABLE(\"&ST_EligibilityDescription\", \"EULA\")"
     TotalNumberOfPlayedMatches:Text="LOCTEXT(\"You must meet the participation requirements by taking part in a minimum number of events.\")"
     SeasonNumberOfPlayedMatches:Text="LOCTEXT(\"You must meet the participation requirements by taking part in a minimum number of events.\")"
     Age:Text="LOCTABLE(\"&ST_EligibilityDescription\", \"AgeLock\")"
     RegionLock:Text="LOCTABLE(\"&ST_EligibilityDescription\", \"RegionLock\")"
     AccountLock:Text="LOCTABLE(\"&ST_EligibilityDescription\", \"AccountLock\")"
     TeamLock:Text="LOCTABLE(\"&ST_EligibilityDescription\", \"TeamLock\")"
     TokenRestrictions:Text=@n2
     SuspiciousBehavior:Text=@n2
     PermanentTournamentBan:Text=@n2
     InputType:Text="LOCTABLE(\"&ST_EligibilityDescription\", \"InputType\")"
     TimeAlive:Text="LOCTEXT(\"You or your teammates must accumulate at least 10 hours time alive in eligible events in the past 180 days.\")"
     PowerRankings:Text="LOCTEXT(\"Your account must meet the required Power Rankings threshold to be eligible for this tournament. Visit fortnite.com/competitive/power-rankings for more information.\")"
     Index=@n1.RequirementType
    ) -> (ReturnValue:Text)
   )
  ]}
 RequirementsHeaderTexts {nodes [
   n1 = RequirementsHeaderTexts!() -> (then, Arguments:"Tournament Requirement Item Arguments Structure (by ref)", RequirementType:EFortTournamentRequirementCheckType Enum=Invalid) {is=title K2Node_FunctionEntry}
   n2 = Break Tournament Requirement Item Arguments!(TournamentRequirementItemArguments:"Tournament Requirement Item Arguments Structure (by ref)"=@n1) -> (
    HabaneroRankCheckResult:EFortCompeteHabaneroRankRequirementCheckResult Enum
    MinimumAccountLevel:int
    TimeAlive:int
    TimeNeedToBeAlive:int
    ValidInputTypes:EFortInputFilterLevel Enum
    NumberMatchesPlayed:int
    NumberOfMatchesNeedToBePlayed:int
    SeasonNumber:int
   ) {is=title K2Node_BreakStruct}
   n3 = "To Text (Float)"(
    Value=SafeDivide(A="To Float (Integer)"(InInt=@n2.TimeAlive), B=3600)
    RoundingMode=ToZero
    MinimumIntegralDigits=1
    MaximumIntegralDigits=324
    MinimumFractionalDigits=0
    MaximumFractionalDigits=1
   )
   n4 = int / int(A=@n2.TimeNeedToBeAlive, B=3600)
   n5 = Select!(
    None:Text
    Touch:Text="LOCTEXT(\"Touch\")"
    Gamepad:Text="LOCTEXT(\"Touch && Gamepad\")"
    Mouse:Text
    Index:EFortInputFilterLevel Enum=@n2.ValidInputTypes
   ) -> (ReturnValue)
   > Return Node(ReturnValue=Select!(
     Invalid:Text="LOCTABLE(\"&/ST_EligibilityTitle\", \"InvalidReqCheck\")"
     SystemFeatureFlag_CryptoProcessor:Text="LOCTABLE(\"&ST_EligibilityTitle\", \"SysCryptoProc\")"
     SystemFeatureFlag_SecureBoot:Text="LOCTABLE(\"&ST_EligibilityTitle\", \"SysSeqBoot\")"
     SystemFeatureFlag_KernelCodeIntegrity:Text="LOCTABLE(\"&ST_EligibilityTitle\", \"SysFlagKernel\")"
     SystemFeatureFlag_IOMMU:Text="LOCTABLE(\"&ST_EligibilityTitle\", \"SysIOMMU\")"
     SystemFeatureFlag_StableOSVersion:Text="LOCTABLE(\"&ST_EligibilityTitle\", \"SysOSVersion\")"
     SystemFeatureFlag_StrictInjectionPolicy:Text="LOCTABLE(\"&ST_EligibilityTitle\", \"SysInjection\")"
     SystemFeatureFlag_StrictVirtualizationPolicy:Text="LOCTABLE(\"&ST_EligibilityTitle\", \"SysNotVM\")"
     SystemFeatureFlag_AndroidSecureDevice:Text="LOCTABLE(\"&ST_EligibilityTitle\", \"SysAndroidSecureDevice\")"
     SystemFeatureFlag_IOSSecureDevice:Text="LOCTABLE(\"&ST_EligibilityTitle\", \"SysIOSSecureDevice\")"
     SystemFeatureFlag_IOSMinimumOSVersion:Text="LOCTABLE(\"&ST_EligibilityTitle\", \"SysIOSMinimumOSVersion\")"
     AccountLevel:Text=Format Text!(
      Format="LOCTABLE(\"&ST_EligibilityTitle\", \"AccountLevel\")"
      level:int=@n2.MinimumAccountLevel
     ) -> (Result)
     MultiFactorAuthentication:Text="LOCTABLE(\"&ST_EligibilityTitle\", \"MFA\")"
     HabaneroRank:Text=Select!(
      Invalid="LOCTABLE(\"&ST_EligibilityTitle\", \"RankInvalid\")"
      Valid:Text="LOCTABLE(\"&ST_EligibilityTitle\", \"RankValid\")"
      PlayerIsUnhabaneroed:Text="LOCTABLE(\"&ST_EligibilityTitle\", \"RankUnhabaneroed\")"
      PlayerCurrentRankBelowMinimumRequirement:Text="LOCTABLE(\"&ST_EligibilityTitle\", \"RankBelowMin\")"
      PlayerHighestRankBelowMinimumRequirement:Text="LOCTABLE(\"&ST_EligibilityTitle\", \"RankBelowMin\")"
      PlayerCurrentRankAboveMaximumRequirement:Text="LOCTABLE(\"&ST_EligibilityTitle\", \"RankAboveMax\")"
      PlayerIsIneligible:Text="LOCTABLE(\"&ST_EligibilityTitle\", \"RankBelowMin\")"
      PlayerTrackRankBelowMinimumRequirement:Text="LOCTABLE(\"&ST_EligibilityTitle\", \"RankBelowMin\")"
      PlayerTrackRankAboveMaximumRequirement:Text="LOCTABLE(\"&ST_EligibilityTitle\", \"RankAboveMax\")"
      Index:EFortCompeteHabaneroRankRequirementCheckResult Enum=@n2
     ) -> (ReturnValue)
     QualifyingRound:Text="LOCTABLE(\"&ST_EligibilityTitle\", \"QualifyingRound\")"
     EULA:Text="LOCTABLE(\"&ST_EligibilityTitle\", \"EULA\")"
     TotalNumberOfPlayedMatches:Text=Format Text!(
      Format="LOCTABLE(\"&ST_EligibilityTitle\", \"TotalNumOfPlayerMatches\")"
      Played:int=@n2.NumberMatchesPlayed
      N:int=@n2.NumberOfMatchesNeedToBePlayed
     ) -> (Result)
     SeasonNumberOfPlayedMatches:Text=Format Text!(
      Format="LOCTABLE(\"&ST_EligibilityTitle\", \"SeasonNumberOfPlayerMatches\")"
      Played=@n2.NumberMatchesPlayed
      n:int=@n2.NumberOfMatchesNeedToBePlayed
      "chapter&&season":int=@n2.SeasonNumber
     ) -> (Result)
     Age:Text="LOCTABLE(\"&ST_EligibilityTitle\", \"AgeLock\")"
     RegionLock:Text="LOCTABLE(\"&ST_EligibilityTitle\", \"RegionLock\")"
     AccountLock:Text="LOCTABLE(\"&ST_EligibilityTitle\", \"AccountLock\")"
     TeamLock:Text="LOCTABLE(\"&ST_EligibilityTitle\", \"TeamLock\")"
     TokenRestrictions:Text="LOCTABLE(\"&ST_EligibilityTitle\", \"NoTokens\")"
     SuspiciousBehavior:Text="LOCTABLE(\"&ST_EligibilityTitle\", \"SuspiciousBehaviour\")"
     PermanentTournamentBan:Text
     InputType:Text=Format Text!(
      Format="LOCTABLE(\"&ST_EligibilityTitle\", \"InputType\")"
      ValidInputTypes:Text=@n5
     ) -> (Result)
     TimeAlive:Text=Format Text!(Format="LOCTEXT(\"Survive for {TimeAlive}/{RequiredTime} hours in Events\")", TimeAlive:Text=@n3, RequiredTime:int=@n4) -> (Result)
     PowerRankings:Text="LOCTEXT(\"Power Rankings\")"
     Index:EFortTournamentRequirementCheckType Enum=@n1.RequirementType
    ) -> (ReturnValue))
  ]}
 SetEligibilityCriteria {nodes [
   n1 = SetEligibilityCriteria!() -> (then, InItems:Array of Tournament Requirement Item Structures) {is=title K2Node_FunctionEntry}
   > Branch(Condition=Is Not Empty!(TargetArray:Array of Tournament Requirement Item Structures=@n1) -> (ReturnValue:bool))
   > ClearListItems(self=Get CriteriaListView())
   > n2 = For Each Loop!(Exec, Array:Array of Tournament Requirement Item Structures=@n1) -> (LoopBody, Array Element:Tournament Requirement Item Structure, Array Index, Completed)
   n3 = Break Tournament Requirement Item!(TournamentRequirementItem:"Tournament Requirement Item Structure (by ref)"=@n2) -> (RequirementType:EFortTournamentRequirementCheckType Enum, bHasPassed:bool=~, Arguments:Tournament Requirement Item Arguments Structure) {is=title K2Node_BreakStruct}
   > n4 = Set RequirementType(RequirementType:EFortTournamentRequirementCheckType Enum=@n3) -> (Output_Get:EFortTournamentRequirementCheckType Enum=Invalid)
   > n5 = Branch(Condition="Equal (Enum)"(A:EFortTournamentRequirementCheckType Enum=@n4, B:EFortTournamentRequirementCheckType Enum=EULA))
   n6 = Branch(
    execute=@n5.else
    Condition=Contains Item(
     TargetArray:Array of EFortTournamentRequirementCheckType Enums=Get SkipCriteriaList() -> (SkipCriteriaList:Array of EFortTournamentRequirementCheckType Enums=Invalid)
     ItemToFind:"EFortTournamentRequirementCheckType Enum (by ref)"=@n4
    )
   )
   n7 = NOT Boolean(A=Get RedMsgSet() -> (RedMsgSet:bool))
   Branch(execute=[@n13.then, @n6.then], Condition=@n7)
   > n8 = Branch(Condition=@n3.bHasPassed)
   > SetErrorText!(execute, self, ErrorText:Text) -> (then)
   n9 = Find(
    TargetMap:Map of EFortTournamentRequirementCheckType Enums to Texts=Get RedMsgList() -> (RedMsgList:Map of EFortTournamentRequirementCheckType Enums to Texts=Invalid)
    Key:"EFortTournamentRequirementCheckType Enum (by ref)"=Get RequirementType() -> (RequirementType:EFortTournamentRequirementCheckType Enum=Invalid)
   ) -> (Value:Text)
   SetErrorText(execute=@n8.else, ErrorText=@n9)
   > Branch(Condition=AND Boolean(A=@n9.ReturnValue, B=@n7))
   > Set RedMsgSet(RedMsgSet:bool=true) -> (Output_Get:bool)
   Set Arguments(execute=@n6.else, Arguments:Tournament Requirement Item Arguments Structure=@n3.Arguments) -> (Output_Get:Tournament Requirement Item Arguments Structure)
   n10 = Get Arguments() -> (Arguments:Tournament Requirement Item Arguments Structure)
   n11 = Get RequirementType() -> (RequirementType=Invalid)
   > n12 = Construct Object from Class!(
    execute
    Class=&/VM_EligibilityCheckCriteria.VM_EligibilityCheckCriteria_C
    self
    bPassed:bool=@n3.bHasPassed
    Name:Text=RequirementsHeaderTexts!(Arguments:"Tournament Requirement Item Arguments Structure (by ref)"=@n10, RequirementType:EFortTournamentRequirementCheckType Enum=@n11) -> (ReturnValue:Text) {is=title K2Node_CallFunction}
    Description:Text=RequirementsDescriptionTexts!(Arguments:"Tournament Requirement Item Arguments Structure (by ref)"=@n10, RequirementType:EFortTournamentRequirementCheckType Enum=@n11) -> (ReturnValue:Text) {is=title K2Node_CallFunction}
   ) -> (then, ReturnValue:VM Eligibility Check Criteria Object Reference)
   > n13 = AddItem!(execute, self:List View Object Reference=Get CriteriaListView(), Item:Object Reference=@n12) -> (then)
  ]}
 SetErrorText {nodes [
   n1 = SetErrorText!() -> (then, ErrorText:Text) {is=title K2Node_FunctionEntry}
   > "SetText (Text)"(self=Get ErrorMessage(), InText=@n1)
   > SetVisibility(
    self=Get Error() -> (Error:Grid Panel Object Reference)
    InVisibility=Select!(Option 0:ESlateVisibility Enum=SelfHitTestInvisible, Option 1:ESlateVisibility Enum=Collapsed, Index:bool=TextIsEmpty(InText=@n1)) -> (ReturnValue:ESlateVisibility Enum)
   )
   > PrintString!(execute, InString=Append(A="TXT: ", B="To String (Text)"(InText=@n1)), bPrintToScreen, bPrintToLog, TextColor, Duration, Key) -> (then)
  ]}
 SetExplanationText {nodes [
   n1 = SetExplanationText!() -> (then, bChecksPassed:bool=~) {is=title K2Node_FunctionEntry}
   > Set Text(
    Text=Select(
     Option 0:Text="LOCTEXT(\"Unable to participate in tournament due to eligibility issues. To ensure that you qualify:\")"
     Option 1:Text="LOCTEXT(\"You qualify to participate in this tournament:\")"
     Index=@n1
    ) -> (ReturnValue:Text)
    self=Get ExplanationText() -> (ExplanationText:Common Text Object Reference)
   )
  ]}
]
variables [
 GameModeRankedDataAsset {type=GameModeRankedDataAsset, default=/Game/UI/Habanero/GameModeRankedDataAsset}
 RedMsgList {default {
   accountLock=Another account has already played in this event.
   permanentTournamentBan=Your account is ineligible for this event.
   regionLock=You have already participated in the maximum number of regions.
   suspiciousBehavior=Your account is ineligible for this event.
   teamLock=You have already played with the maximum number of teams.
  }}
 SkipCriteriaList {type="array<EFortTournamentRequirementCheckType>", default [PermanentTournamentBan]}
]
widgets {
 info {
  inheritedWidgetCount=0
  namedSlotCount=1
  parentClass=&CommonActivatableWidget
  rootWidgetClass=/FortUILibrary/Systems/Dialog/WBP_UIKit_Dialog_Internal.WBP_UIKit_Dialog_Internal_C
  widgetCount=10
 }
 widgets [
  Dialog_Base {
   widgetClassPath=/FortUILibrary/Systems/Dialog/WBP_UIKit_Dialog_Internal
   bIsVariable=true
   widget=Dialog_Base
   properties {
    bIsFocusable=true
    navigation=&self:WidgetTree.Dialog_Base.WidgetNavigation_0
    buttons {
     each Button Horizontal Fill=true
     each Button Min Width=0
     each Button Size=Medium
     focuse Button=0
     space Between Buttons=30
     cTA Buttons Types [Quiet]
     group Placement {
      horizontal Alignment=HAlign_Fill
      vertical Alignment=VAlign_Bottom
      grid {column=1, column Span=0, layer=0, row=2, row Span=0, nudge=<0 0>}
      padding=<ltrb 80 64 80 48>
     }
    }
    dialogSIze {
     height Override {enable=false, value=800}
     max Aspect Ratio {enable=false, value=0}
     max Desired Height {enable=true, value=1000}
     max Desired Width {enable=false, value=0}
     min Aspect Ratio {enable=false, value=0}
     min Desired Height {enable=false, value=0}
     min Desired Width {enable=false, value=0}
     width Override {enable=true, value=1296}
    }
    textProperties {
     auto Wrap Text=true
     justification=Left
     line Height Percentage=1.3
     mobile Font Size Multiplier=1.3
     transform Policy=None
     wrap Text At=0
     color {colorUseRule=UseColor_Specified, specifiedColor=<rgba 1 1 1 0.8>}
     font {
      bForceMonospaced=false
      bMaterialIsStencil=false
      fontMaterial=null
      fontObject=&HeadingNow
      letterSpacing=0
      monospacedWidth=1
      size=15.75
      skewAmount=0
      typefaceFontName=74 Regular
      outlineSettings {
       bApplyOutlineToDropShadows=false
       bMiteredCorners=false
       bSeparateFillAlpha=false
       outlineMaterial=null
       outlineSize=0
       outlineColor=<rgba 0 0 0 1>
      }
     }
     placement {
      horizontal Alignment=HAlign_Fill
      vertical Alignment=VAlign_Top
      grid {column=1, column Span=0, layer=0, row=1, row Span=0, nudge=<0 0>}
      padding=<ltrb 80 0 80 0>
     }
    }
    titleProperties {
     auto Wrap Text=true
     justification=Left
     line Height Percentage=1
     mobile Font Size Multiplier=1.3
     transform Policy=ToUpper
     wrap Text At=0
     color {colorUseRule=UseColor_Specified, specifiedColor=<rgba 1 1 1 1>}
     font {
      bForceMonospaced=false
      bMaterialIsStencil=false
      fontMaterial=null
      fontObject=&HeadingNow
      letterSpacing=0
      monospacedWidth=1
      size=25.5
      skewAmount=0
      typefaceFontName=86 Bold Italic
      outlineSettings {
       bApplyOutlineToDropShadows=false
       bMiteredCorners=false
       bSeparateFillAlpha=false
       outlineMaterial=null
       outlineSize=0
       outlineColor=<rgba 0 0 0 1>
      }
     }
     placement {
      horizontal Alignment=HAlign_Fill
      vertical Alignment=VAlign_Top
      grid {column=1, column Span=0, layer=0, row=0, row Span=0, nudge=<0 0>}
      padding=<ltrb 80 64 80 16>
     }
    }
   }
  }
  MainRoot {widgetClassPath=&GridPanel, bIsVariable=true, namedSlotHost=Dialog_Base, widget=MainRoot, properties {columnFill [1]}}
  Error {
   widgetClassPath=&GridPanel
   bIsVariable=true
   parent=MainRoot
   widget=Error
   slot {class=GridSlot, properties {padding=<ltrb 0 10 0 18>}}
  }
  ErrorMessage {
   widgetClassPath=&CommonTextBlock
   bIsVariable=true
   parent=Error
   widget=ErrorMessage
   properties {
    autoWrapText=true
    bAutoCollapseWithEmptyText=true
    text=This device has already been used by another account to participate in this event.
    visibility=Collapsed
    colorAndOpacity {colorUseRule=UseColor_Specified, specifiedColor=<rgba 1 0.147 0.1329 1>}
    font {
     bForceMonospaced=false
     bMaterialIsStencil=false
     fontMaterial=null
     fontObject=&HeadingNow
     letterSpacing=0
     monospacedWidth=1
     size=18
     skewAmount=0
     typefaceFontName=74 Regular
     outlineSettings {
      bApplyOutlineToDropShadows=false
      bMiteredCorners=false
      bSeparateFillAlpha=false
      outlineMaterial=null
      outlineSize=0
      outlineColor=<rgba 0 0 0 1>
     }
    }
    shadowOffset=<0 0>
   }
   slot {class=GridSlot, properties {column=1, padding=<ltrb 12 0 0 0>}}
  }
  AlertImage {
   widgetClassPath=&Image
   bIsVariable=true
   parent=Error
   widget=AlertImage
   properties {
    brush {
     bIsDynamicallyLoaded=false
     drawAs=Image
     imageType=FullColor
     mirroring=NoMirror
     resourceName=None
     resourceObject=/FortUILibrary/Textures/MaterialInstances/MI_UIKit_Icon_AgeGatedAlert
     tiling=NoTile
     imageSize=<32 32>
     margin=<ltrb 0 0 0 0>
     outlineSettings {
      bUseBrushTransparency=false
      roundingType=HalfHeightRadius
      width=0
      color {colorUseRule=UseColor_Specified, specifiedColor=<rgba 0 0 0 0>}
      cornerRadii=<0 0 0 0>
     }
     tintColor {colorUseRule=UseColor_Specified, specifiedColor=<rgba 1 1 1 1>}
     uVRegion {bIsValid=false, max=<0 0>, min=<0 0>}
    }
    colorAndOpacity=<rgba 1 0.147 0.1329 1>
   }
   slot {class=GridSlot, properties {horizontalAlignment=HAlign_Center, verticalAlignment=VAlign_Center}}
  }
  ExplanationText {
   widgetClassPath=&CommonTextBlock
   bIsVariable=true
   parent=MainRoot
   widget=ExplanationText
   properties {
    autoWrapText=true
    text=Unable to participate in tournament due to eligibility issues. To ensure that you qualify:
    colorAndOpacity {colorUseRule=UseColor_Specified, specifiedColor=<rgba 0.5776 0.5776 0.5776 1>}
    font {
     bForceMonospaced=false
     bMaterialIsStencil=false
     fontMaterial=null
     fontObject=&HeadingNow
     letterSpacing=0
     monospacedWidth=1
     size=18
     skewAmount=0
     typefaceFontName=74 Regular
     outlineSettings {
      bApplyOutlineToDropShadows=false
      bMiteredCorners=false
      bSeparateFillAlpha=false
      outlineMaterial=null
      outlineSize=0
      outlineColor=<rgba 0 0 0 1>
     }
    }
    shadowOffset=<0 0>
   }
   slot {class=GridSlot, properties {row=1}}
  }
  SizeBox_0 {
   widgetClassPath=&SizeBox
   parent=MainRoot
   widget=SizeBox_0
   properties {bOverride_MaxDesiredHeight=true, maxDesiredHeight=424}
   slot {class=GridSlot, properties {row=2, padding=<ltrb 10 30 0 0>}}
  }
  CriteriaListView {
   widgetClassPath=&CommonListView
   bIsVariable=true
   parent=SizeBox_0
   widget=CriteriaListView
   properties {
    allowOverscroll=false
    entryWidgetClass=&/WBP_EligibilityCriterion.WBP_EligibilityCriterion_C
    numDesignerPreviewEntries=9
    selectionMode=None
    verticalEntrySpacing=35
    scrollBarStyle {
     thickness=12
     draggedThumbImage {
      bIsDynamicallyLoaded=false
      drawAs=Box
      imageType=FullColor
      mirroring=NoMirror
      resourceName=../../../Engine/Content/Slate/Common/Scrollbar_Thumb.png
      resourceObject=null
      tiling=NoTile
      imageSize=<0 0>
      margin=<ltrb 0.25 0.25 0.25 0.25>
      outlineSettings {
       bUseBrushTransparency=false
       roundingType=HalfHeightRadius
       width=0
       color {colorUseRule=UseColor_Specified, specifiedColor=<rgba 0 0 0 0>}
       cornerRadii=<0 0 0 0>
      }
      tintColor {colorUseRule=UseColor_Specified, specifiedColor=<rgba 1 1 1 1>}
      uVRegion {bIsValid=false, max=<0 0>, min=<0 0>}
     }
     horizontalBackgroundImage {
      bIsDynamicallyLoaded=false
      drawAs=NoDrawType
      imageType=NoImage
      mirroring=NoMirror
      resourceName=None
      resourceObject=null
      tiling=NoTile
      imageSize=<0 0>
      margin=<ltrb 0 0 0 0>
      outlineSettings {
       bUseBrushTransparency=false
       roundingType=HalfHeightRadius
       width=0
       color {colorUseRule=UseColor_Specified, specifiedColor=<rgba 0 0 0 0>}
       cornerRadii=<0 0 0 0>
      }
      tintColor {colorUseRule=UseColor_Specified, specifiedColor=<rgba 1 1 1 1>}
      uVRegion {bIsValid=false, max=<0 0>, min=<0 0>}
     }
     horizontalBottomSlotImage {
      bIsDynamicallyLoaded=false
      drawAs=Image
      imageType=FullColor
      mirroring=NoMirror
      resourceName=../../../Engine/Content/Slate/Common/Scrollbar_Background_Horizontal.png
      resourceObject=null
      tiling=NoTile
      imageSize=<8 8>
      margin=<ltrb 0 0 0 0>
      outlineSettings {
       bUseBrushTransparency=false
       roundingType=HalfHeightRadius
       width=0
       color {colorUseRule=UseColor_Specified, specifiedColor=<rgba 0 0 0 0>}
       cornerRadii=<0 0 0 0>
      }
      tintColor {colorUseRule=UseColor_Specified, specifiedColor=<rgba 1 1 1 1>}
      uVRegion {bIsValid=false, max=<0 0>, min=<0 0>}
     }
     horizontalTopSlotImage {
      bIsDynamicallyLoaded=false
      drawAs=Image
      imageType=FullColor
      mirroring=NoMirror
      resourceName=../../../Engine/Content/Slate/Common/Scrollbar_Background_Horizontal.png
      resourceObject=null
      tiling=NoTile
      imageSize=<8 8>
      margin=<ltrb 0 0 0 0>
      outlineSettings {
       bUseBrushTransparency=false
       roundingType=HalfHeightRadius
       width=0
       color {colorUseRule=UseColor_Specified, specifiedColor=<rgba 0 0 0 0>}
       cornerRadii=<0 0 0 0>
      }
      tintColor {colorUseRule=UseColor_Specified, specifiedColor=<rgba 1 1 1 1>}
      uVRegion {bIsValid=false, max=<0 0>, min=<0 0>}
     }
     hoveredThumbImage {
      bIsDynamicallyLoaded=false
      drawAs=Box
      imageType=FullColor
      mirroring=NoMirror
      resourceName=../../../Engine/Content/Slate/Common/Scrollbar_Thumb.png
      resourceObject=null
      tiling=NoTile
      imageSize=<0 0>
      margin=<ltrb 0.25 0.25 0.25 0.25>
      outlineSettings {
       bUseBrushTransparency=false
       roundingType=HalfHeightRadius
       width=0
       color {colorUseRule=UseColor_Specified, specifiedColor=<rgba 0 0 0 0>}
       cornerRadii=<0 0 0 0>
      }
      tintColor {colorUseRule=UseColor_Specified, specifiedColor=<rgba 1 1 1 1>}
      uVRegion {bIsValid=false, max=<0 0>, min=<0 0>}
     }
     normalThumbImage {
      bIsDynamicallyLoaded=false
      drawAs=Box
      imageType=FullColor
      mirroring=NoMirror
      resourceName=../../../Engine/Content/Slate/Common/Scrollbar_Thumb.png
      resourceObject=null
      tiling=NoTile
      imageSize=<0 0>
      margin=<ltrb 0.25 0.25 0.25 0.25>
      outlineSettings {
       bUseBrushTransparency=false
       roundingType=HalfHeightRadius
       width=0
       color {colorUseRule=UseColor_Specified, specifiedColor=<rgba 0 0 0 0>}
       cornerRadii=<0 0 0 0>
      }
      tintColor {colorUseRule=UseColor_Specified, specifiedColor=<rgba 1 1 1 1>}
      uVRegion {bIsValid=false, max=<0 0>, min=<0 0>}
     }
     verticalBackgroundImage {
      bIsDynamicallyLoaded=false
      drawAs=NoDrawType
      imageType=NoImage
      mirroring=NoMirror
      resourceName=None
      resourceObject=null
      tiling=NoTile
      imageSize=<0 0>
      margin=<ltrb 0 0 0 0>
      outlineSettings {
       bUseBrushTransparency=false
       roundingType=HalfHeightRadius
       width=0
       color {colorUseRule=UseColor_Specified, specifiedColor=<rgba 0 0 0 0>}
       cornerRadii=<0 0 0 0>
      }
      tintColor {colorUseRule=UseColor_Specified, specifiedColor=<rgba 1 1 1 1>}
      uVRegion {bIsValid=false, max=<0 0>, min=<0 0>}
     }
     verticalBottomSlotImage {
      bIsDynamicallyLoaded=false
      drawAs=Image
      imageType=FullColor
      mirroring=NoMirror
      resourceName=../../../Engine/Content/Slate/Common/Scrollbar_Background_Vertical.png
      resourceObject=null
      tiling=NoTile
      imageSize=<8 8>
      margin=<ltrb 0 0 0 0>
      outlineSettings {
       bUseBrushTransparency=false
       roundingType=HalfHeightRadius
       width=0
       color {colorUseRule=UseColor_Specified, specifiedColor=<rgba 0 0 0 0>}
       cornerRadii=<0 0 0 0>
      }
      tintColor {colorUseRule=UseColor_Specified, specifiedColor=<rgba 1 1 1 1>}
      uVRegion {bIsValid=false, max=<0 0>, min=<0 0>}
     }
     verticalTopSlotImage {
      bIsDynamicallyLoaded=false
      drawAs=Image
      imageType=FullColor
      mirroring=NoMirror
      resourceName=../../../Engine/Content/Slate/Common/Scrollbar_Background_Vertical.png
      resourceObject=null
      tiling=NoTile
      imageSize=<8 8>
      margin=<ltrb 0 0 0 0>
      outlineSettings {
       bUseBrushTransparency=false
       roundingType=HalfHeightRadius
       width=0
       color {colorUseRule=UseColor_Specified, specifiedColor=<rgba 0 0 0 0>}
       cornerRadii=<0 0 0 0>
      }
      tintColor {colorUseRule=UseColor_Specified, specifiedColor=<rgba 1 1 1 1>}
      uVRegion {bIsValid=false, max=<0 0>, min=<0 0>}
     }
    }
   }
   slot {class=SizeBoxSlot}
  }
  DetailsText_1 {
   widgetClassPath=&CommonTextBlock
   parent=MainRoot
   widget=DetailsText_1
   properties {
    autoWrapText=true
    text=Participation is only allowed for those who have accepted EULA on the first attempt to start tournament matchmaking.
    colorAndOpacity {colorUseRule=UseColor_Specified, specifiedColor=<rgba 0.5776 0.5776 0.5776 1>}
    font {
     bForceMonospaced=false
     bMaterialIsStencil=false
     fontMaterial=null
     fontObject=&HeadingNow
     letterSpacing=0
     monospacedWidth=1
     size=18
     skewAmount=0
     typefaceFontName=74 Regular
     outlineSettings {
      bApplyOutlineToDropShadows=false
      bMiteredCorners=false
      bSeparateFillAlpha=false
      outlineMaterial=null
      outlineSize=0
      outlineColor=<rgba 0 0 0 1>
     }
    }
    shadowOffset=<0 0>
   }
   slot {class=GridSlot, properties {row=6, padding=<ltrb 0 32 0 0>}}
  }
  DetailsText {
   widgetClassPath=&CommonTextBlock
   parent=MainRoot
   widget=DetailsText
   properties {
    autoWrapText=true
    text=For details see: www.epicgames.com/fortnite/competitive/rules-library.
    colorAndOpacity {colorUseRule=UseColor_Specified, specifiedColor=<rgba 0.5776 0.5776 0.5776 1>}
    font {
     bForceMonospaced=false
     bMaterialIsStencil=false
     fontMaterial=null
     fontObject=&HeadingNow
     letterSpacing=0
     monospacedWidth=1
     size=18
     skewAmount=0
     typefaceFontName=74 Regular
     outlineSettings {
      bApplyOutlineToDropShadows=false
      bMiteredCorners=false
      bSeparateFillAlpha=false
      outlineMaterial=null
      outlineSize=0
      outlineColor=<rgba 0 0 0 1>
     }
    }
    shadowOffset=<0 0>
   }
   slot {class=GridSlot, properties {row=7, padding=<ltrb 0 30 0 0>}}
  }
 ]
}
